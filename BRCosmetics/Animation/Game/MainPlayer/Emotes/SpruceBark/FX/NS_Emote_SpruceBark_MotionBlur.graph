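# Niagara system for the SpruceBark player emote that fakes spinning motion blur with short-lived, rotating mesh rings around the body.
# BlurRings_Detail and BlurRings_Thick use spawn groups to set color, scale, lifetime and height band per ring layer, with Thick adding a glow color.
# VortexWind adds a faint swirl near the base, while Light adds warm, flickering lights around upper-body height.
# Time-of-day functions adjust the rings' and wind's brightness.
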NiagaraSystem /BRCosmetics/Animation/Game/MainPlayer/Emotes/SpruceBark/FX/NS_Emote_SpruceBark_MotionBlur
dependencies {
 usedDataInterfaces [&NiagaraDataInterfaceCurve, &NiagaraDataInterfaceArrayColor, &NiagaraDataInterfaceColorCurve]
 usedDynamicInputs [&MakeVector]
 usedModules [&SystemState]
}
emitters {
 BlurRings_Detail {
  emitterUpdateScript [
   &EmitterState()
   &SpawnRate(SpawnRate=&Multiply_Float(A=30, B=1), Use Spawn Probability=false)
   &SpawnRate(SpawnGroup=1, SpawnRate=&Multiply_Float(A=30, B=1), Use Spawn Probability=false) {module=SpawnRate001}
   &SpawnRate(SpawnGroup=2, SpawnRate=&Multiply_Float(A=35, B=1), Use Spawn Probability=false) {module=SpawnRate002}
   &SpawnRate(SpawnGroup=3, SpawnRate=&Multiply_Float(A=15, B=1), Use Spawn Probability=false) {module=SpawnRate003}
   &SpawnRate(SpawnGroup=4, SpawnRate=&Multiply_Float(A=15, B=1), Use Spawn Probability=false) {module=SpawnRate004}
  ]
  particleSpawnScript [
   &InitializeParticle(
    Color=&SelectLinearColorFromArray(
     Array Sampling Mode=::NewEnumerator1
     Direct Array Index=&SpawnGroup()
     Color Selection Array=DataInterface() {propertyValues {ColorData [
        <rgba 0.3698 0.097 0.0032 1>
        <rgba 0.3698 0.097 0.0032 1>
        <rgba 0.3698 0.097 0.0032 1>
        <rgba 0.3698 0.097 0.0032 1>
        <rgba 0.0783 0.0073 0 1>
       ]}}
     Random Seed=?
    )
    Lifetime=&MergeHelper_FloatSwitch(0=0.27, 1=0.23, 2=0.24, 3=0.32, 4=0.32, Index=&SpawnGroup(), Num=::NewEnumerator3)
    Mesh Scale Max=&Multiply_Vector(
     A=<1 1 0.6>
     B=&MergeHelper_VectorSwitch(
      0=<1.5 1.5 3>
      1=<2.3 2.3 1>
      2=<2.1 2.1 1.5>
      3=<1 1 6>
      4=<0.9 0.9 5>
      Index=&SpawnGroup()
      Num=::NewEnumerator3
     )
    )
    Mesh Scale Min=&Multiply_Vector(
     A=<1 1 0.3>
     B=&MergeHelper_VectorSwitch(
      0=<1 1 3>
      1=<2.2 2.2 1>
      2=<1.9 1.9 1.5>
      3=<0.8 0.8 5>
      4=<0.7 0.7 4>
      Index=&SpawnGroup()
      Num=::NewEnumerator3
     )
    )
    Mesh Scale Mode=::NewEnumerator2
    Position Offset=&MergeHelper_VectorSwitch(
     0=&RandomRangeVector(Minimum=<0 0 95>, Maximum=<0 0 142>)
     1=&RandomRangeVector(Minimum=<0 0 130>, Maximum=<0 0 133>)
     2=&RandomRangeVector(Minimum=<0 0 112>, Maximum=<0 0 140>)
     3=&RandomRangeVector(Minimum=<0 0 50>, Maximum=<0 0 90>)
     4=&RandomRangeVector(Minimum=<0 0 10>, Maximum=<0 0 60>)
     Index=&SpawnGroup()
     Num=::NewEnumerator3
    )
    UsePositionOffset=true
    Write Material Random=true
   )
   &CameraOffset(Camera Offset Amount=15)
   &InitialMeshOrientation(
    Mesh Axis To Orient=<0 0 1>
    Mesh Orientation Mode=::NewEnumerator2
    Orientation Coordinate Space=::World
    Orientation Vector=<0 0 1>
    Rotation=&MergeHelper_VectorSwitch(
     0=&RandomRangeVector(Minimum=<-0.02 -0.02 0>, Maximum=<0.02 0.02 0>)
     1=&RandomRangeVector(Minimum=<-0.015 -0.015 0>, Maximum=<0.015 0.015 0>)
     2=&RandomRangeVector(Minimum=<-0.02 -0.02 0>, Maximum=<0.02 0.02 0>)
     3=&RandomRangeVector(Minimum=<-0.03 -0.03 0>, Maximum=<0.03 0.03 0>)
     4=&RandomRangeVector(Minimum=<-0.03 -0.03 0>, Maximum=<0.03 0.03 0>)
     Index=&SpawnGroup()
     Num=::NewEnumerator3
    )
    Use Rotation Vector=true
   )
  ]
  particleUpdateScript [
   &ParticleState()
   &UpdateMeshOrientation(Rotation Vector=<0 0 -4>) {module=UpdateMeshOrientation002}
   &ScaleMeshSize(Scale Factor=&VectorFromFloat(Value=&FloatFromCurve(FloatCurve=DataInterface() {propertyValues {Curve {keys [
          {arriveTangent=1.0772, interpMode=RCIM_Cubic, leaveTangent=1.0772, tangentMode=RCTM_User, time=0, value=0.7}
          {arriveTangent=1.0745, interpMode=RCIM_Cubic, leaveTangent=1.0745, tangentMode=RCTM_User, time=0.2, value=0.9118}
          {arriveTangentWeight=0.1434, interpMode=RCIM_Cubic, tangentMode=RCTM_User, time=0.6, value=1}
         ]}}})))
   &SolveForcesAndVelocity(Clamp Velocity=false, Limit Acceleration=false)
   &ScaleColor(
    Scale Alpha=&FloatFromCurve(
     FloatCurve=DataInterface() {propertyValues {Curve {keys [
         {arriveTangent=2.752, interpMode=RCIM_Cubic, leaveTangent=2.752, tangentMode=RCTM_User, time=0, value=0}
         {interpMode=RCIM_Cubic, tangentMode=RCTM_User, time=0.5, value=0.5}
         {arriveTangent=-2.5695, tangentMode=RCTM_User, time=1, value=0}
        ]}}}
     Scale Curve=&FNTimeOfDayScalar(Night Scale=0.5, Morning Scale=0.5, Day Scale=0.5, Evening Scale=0.5)
    )
    Scale RGB=&MakeVectorFromLinearColorRGB(Color=&FNTimeOfDayColor(
      Night Color=<rgba 0.6771 0.6414 0.6171 1>
      Morning Color=<rgba 0.6771 0.6414 0.6171 1>
      Day Color=<rgba 1 1 1 1>
      Evening Color=<rgba 0.6771 0.6414 0.6171 1>
     ))
    ScaleA=true
    ScaleRGB=true
   )
  ]
  rendererClasses [&NiagaraMeshRendererProperties]
  renderers [{rendererClass=&NiagaraMeshRendererProperties, rendererIndex=0}]
 }
 BlurRings_Thick {
  emitterUpdateScript [
   &EmitterState()
   &SpawnBurst_Instantaneous(Spawn Count=20, Use Spawn Probability=false, UseLoopCountLimit=false)
   &SpawnRate(SpawnRate=&Multiply_Float(A=30, B=1), Use Spawn Probability=false)
   &SpawnRate(SpawnGroup=1, SpawnRate=&Multiply_Float(A=35, B=1), Use Spawn Probability=false) {module=SpawnRate001}
   &SpawnRate(SpawnGroup=2, SpawnRate=&Multiply_Float(A=35, B=1), Use Spawn Probability=false) {module=SpawnRate002}
   &SpawnRate(SpawnGroup=3, SpawnRate=&Multiply_Float(A=25, B=1), Use Spawn Probability=false) {module=SpawnRate003}
   &SpawnRate(SpawnGroup=4, SpawnRate=&Multiply_Float(A=20, B=1), Use Spawn Probability=false) {module=SpawnRate004}
   &SpawnRate(SpawnGroup=5, SpawnRate=20, Use Spawn Probability=false) {module=SpawnRate005}
   &SpawnRate(SpawnGroup=6, SpawnRate=20, Use Spawn Probability=false) {module=SpawnRate006}
  ]
  particleSpawnScript [
   &InitializeParticle(
    Color=&SelectLinearColorFromArray(
     Array Sampling Mode=::NewEnumerator1
     Direct Array Index=&SpawnGroup()
     Color Selection Array=DataInterface() {propertyValues {ColorData [
        <rgba 0.4 0.1049 0.0034 1>
        <rgba 1.3 0.612 0.1251 1>
        <rgba 0.724 0.1899 0.0062 1>
        <rgba 0.2135 0.0621 0.01 1>
        <rgba 0.151 0.0433 0.0323 1>
        <rgba 0.4 0.1049 0.0034 0.65>
        <rgba 0.4 0.1049 0.0034 0.5>
       ]}}
     Random Seed=?
    )
    Lifetime=&MergeHelper_FloatSwitch(0=0.24, 1=0.2, 2=0.21, 3=0.29, 4=0.29, Index=&SpawnGroup(), Num=::NewEnumerator3)
    Mesh Scale Max=&Multiply_Vector(
     A=<1 1 0.6>
     B=&MergeHelper_VectorSwitch(
      0=<0.5 0.5 1>
      1=<1.1 1.1 1>
      2=<0.8 0.8 1.5>
      3=<0.5 0.5 1.5>
      4=<0.4 0.4 2>
      5=<0.5 0.5 0.5>
      6=<0.35 0.35 0.5>
      Index=&SpawnGroup()
      Num=::NewEnumerator5
     )
    )
    Mesh Scale Min=&Multiply_Vector(
     A=<1 1 0.3>
     B=&MergeHelper_VectorSwitch(
      0=<0.4 0.4 0.5>
      1=<1 1 1>
      2=<0.8 0.8 1.2>
      3=<0.4 0.4 1.2>
      4=<0.3 0.3 1.5>
      5=<0.4 0.4 1>
      6=<0.25 0.25 1>
      Index=&SpawnGroup()
      Num=::NewEnumerator5
     )
    )
    Mesh Scale Mode=::NewEnumerator2
    Position Offset=&MergeHelper_VectorSwitch(
     0=&RandomRangeVector(Minimum=<0 0 85>, Maximum=<0 0 142>)
     1=&RandomRangeVector(Minimum=<0 0 130>, Maximum=<0 0 133>)
     2=&RandomRangeVector(Minimum=<0 0 105>, Maximum=<0 0 140>)
     3=&RandomRangeVector(Minimum=<0 0 50>, Maximum=<0 0 90>)
     4=&RandomRangeVector(Minimum=<0 0 10>, Maximum=<0 0 60>)
     5=&RandomRangeVector(Minimum=<0 0 142>, Maximum=<0 0 150>)
     6=&RandomRangeVector(Minimum=<0 0 150>, Maximum=<0 0 165>)
     Index=&SpawnGroup()
     Num=::NewEnumerator5
    )
    UsePositionOffset=true
    Write Material Random=true
   )
   &CameraOffset(Camera Offset Amount=15)
   &InitialMeshOrientation(
    Mesh Axis To Orient=<0 0 1>
    Mesh Orientation Mode=::NewEnumerator2
    Orientation Coordinate Space=::World
    Orientation Vector=<0 0 1>
    Rotation=&MergeHelper_VectorSwitch(
     0=&RandomRangeVector(Minimum=<-0.02 -0.02 0>, Maximum=<0.02 0.02 1>)
     1=&RandomRangeVector(Minimum=<-0.015 -0.015 0>, Maximum=<0.015 0.015 1>)
     2=&RandomRangeVector(Minimum=<-0.02 -0.02 0>, Maximum=<0.02 0.02 1>)
     3=&RandomRangeVector(Minimum=<-0.03 -0.03 0>, Maximum=<0.03 0.03 1>)
     4=&RandomRangeVector(Minimum=<-0.03 -0.03 0>, Maximum=<0.03 0.03 1>)
     Index=&SpawnGroup()
     Num=::NewEnumerator3
    )
    Use Rotation Vector=true
   )
   /SetVariables(Particles.Glow Color=&SelectLinearColorFromArray(
     Array Sampling Mode=::NewEnumerator1
     Direct Array Index=&SpawnGroup()
     Color Selection Array=DataInterface() {propertyValues {ColorData [
        <rgba 0.3542 0.0583 0 1>
        <rgba 1 0.3081 0 1>
        <rgba 0.5312 0.0821 0 1>
        <rgba 0.3229 0.0826 0 0.5>
        <rgba 0.0781 0.0224 0.0167 0>
        <rgba 1 0.3333 0.0365 0>
        <rgba 0 0 0 0>
        <rgba 0 0 0 0>
       ]}}
     Random Seed=?
    )) {isSetParameters=true}
  ]
  particleUpdateScript [
   &ParticleState()
   &UpdateMeshOrientation(Rotation Vector=<0 0 -4>) {module=UpdateMeshOrientation002}
   &ScaleMeshSize(Scale Factor=&VectorFromFloat(Value=&FloatFromCurve(FloatCurve=DataInterface() {propertyValues {Curve {keys [
          {arriveTangent=1.0772, interpMode=RCIM_Cubic, leaveTangent=1.0772, tangentMode=RCTM_User, time=0, value=0.4}
          {arriveTangentWeight=0.1434, interpMode=RCIM_Cubic, tangentMode=RCTM_User, time=1, value=1}
         ]}}})))
   &DynamicMaterialParameters(
    Index 0 Param 1=&FloatFromCurve(FloatCurve=DataInterface() {propertyValues {Curve {keys [
         {arriveTangent=2.4535, interpMode=RCIM_Cubic, leaveTangent=2.4535, tangentMode=RCTM_User, time=0, value=0}
         {interpMode=RCIM_Cubic, tangentMode=RCTM_User, time=0.5, value=0.5}
         {arriveTangent=-2.5865, tangentMode=RCTM_User, time=1, value=0}
        ]}}})
    Index 0 Param 2=&FloatFromCurve(FloatCurve=DataInterface() {propertyValues {Curve {keys [<at 0 1>,<at 1 0>]}}}, Scale Curve=3)
    Index 0 Param 3=&FloatFromCurve(FloatCurve=DataInterface() {propertyValues {Curve {defaultValue=1, keys [{tangentMode=RCTM_User, time=0, value=0}, {tangentMode=RCTM_User, time=1, value=1}]}}})
    Index 0 Param 4=1
    Module.Index0Param0Write=false
    Module.Index0Param1Write=false
    Module.Index0Param2Write=false
    Module.Index0Param3Write=true
   )
   &SolveForcesAndVelocity(Clamp Velocity=false, Limit Acceleration=false)
   &ScaleColor(
    Scale Alpha=&FloatFromCurve(
     FloatCurve=DataInterface() {propertyValues {Curve {keys [
         {arriveTangent=5.2303, interpMode=RCIM_Cubic, leaveTangent=5.2303, tangentMode=RCTM_User, time=0, value=0}
         {interpMode=RCIM_Cubic, tangentMode=RCTM_User, time=0.5, value=0.9}
         {arriveTangent=-2.5695, tangentMode=RCTM_User, time=1, value=0}
        ]}}}
     Scale Curve=&FNTimeOfDayScalar(Night Scale=0.6, Morning Scale=0.6, Day Scale=0.7, Evening Scale=0.6)
    )
    Scale RGB=&MakeVectorFromLinearColorRGB(Color=&FNTimeOfDayColor(
      Night Color=<rgba 0.6771 0.6414 0.6171 1>
      Morning Color=<rgba 0.6771 0.6414 0.6171 1>
      Day Color=<rgba 1 1 1 1>
      Evening Color=<rgba 0.6771 0.6414 0.6171 1>
     ))
    ScaleA=true
    ScaleRGB=true
   )
   /SetVariables(Particles.Glow Color=&MultiplyLinearColor(
     Linear Color=$Particles.Initial.Glow Color
     Scale Factor=&MakeVector4FromVectorAndFloat(
      "Vector (RGB)"=&MakeVectorFromLinearColorRGB(Color=&FNTimeOfDayColor(
        Night Color=<rgba 0.6771 0.6414 0.6171 1>
        Morning Color=<rgba 0.6771 0.6414 0.6171 1>
        Day Color=<rgba 1 1 1 1>
        Evening Color=<rgba 0.6771 0.6414 0.6171 1>
       ))
      "Float (Alpha)"=&FloatFromCurve(
       FloatCurve=DataInterface() {propertyValues {Curve {keys [
           {arriveTangent=3.377, interpMode=RCIM_Cubic, leaveTangent=3.377, tangentMode=RCTM_User, time=0, value=0}
           {interpMode=RCIM_Cubic, tangentMode=RCTM_User, time=0.5, value=1}
           {arriveTangent=-2.5695, tangentMode=RCTM_User, time=1, value=0}
          ]}}}
       Scale Curve=&FNTimeOfDayScalar(Night Scale=0.5, Morning Scale=0.6, Day Scale=0.8, Evening Scale=0.6)
      )
     )
    )) {isSetParameters=true}
  ]
  rendererClasses [&NiagaraMeshRendererProperties]
  renderers [
   {rendererClass=&NiagaraMeshRendererProperties, rendererIndex=0}
   {rendererClass=&NiagaraMeshRendererProperties, rendererIndex=1}
  ]
 }
 Light {
  emitterUpdateScript [
   &EmitterState()
   &SpawnBurst_Instantaneous(Spawn Count=2, Spawn Group=1, Use Spawn Probability=false, UseLoopCountLimit=false) {module=SpawnBurst_Instantaneous001}
   &SpawnBurst_Instantaneous(Use Spawn Probability=false, UseLoopCountLimit=false)
  ]
  particleSpawnScript [
   &PartitionParticles()
   &InitializeParticle(
    Color=<rgba 0.3177 0.0766 0 0.5>
    Lifetime=0.65
    Mesh Scale=&Multiply_VectorByFloat(Vector=<2.5 2.5 0.5>, Float=0.48)
    Mesh Scale Mode=::NewEnumerator4
    Position Offset=<0 0 118.8345>
    Sprite Size Mode=::NewEnumerator3
    Uniform Sprite Size=181
    UsePositionOffset=true
    Write Material Random=true
   )
   &CameraOffset(Camera Offset Amount=&MergeHelper_FloatSwitch(0=100, 1=0, 2=?, 3=?, Index=&SpawnGroup(), Num=::NewEnumerator0))
   /SetVariables(Particles.VisibilityTag=&SpawnGroup()) {isSetParameters=true}
  ]
  particleUpdateScript [
   &ParticleState()
   &ScaleColor(
    Linear Color Curve=DataInterface() {propertyValues {
      AlphaCurve {keys [<at 0 0>,<at 0.2528 0.8191>,<at 0.7752 1>,<at 1 0>]}
      BlueCurve {keys [<at 0 1>,<at 1 1>]}
      GreenCurve {keys [<at 0 1>,<at 1 1>]}
      RedCurve {keys [<at 0 1>,<at 1 1>]}
     }}
    Scale Mode=2
   )
   &FN_Light_Attributes(
    Color=&ColorFromCurve(ColorCurve=DataInterface() {propertyValues {
       AlphaCurve {keys [<at 0 0>,<at 0.2528 7>,<at 0.7752 7>,<at 1 0>]}
       BlueCurve {keys [<at 0 0>]}
       GreenCurve {keys [<at 0 2.4646>]}
       RedCurve {keys [<at 0 8>]}
      }})
    Create Color Attribute=true
    Create Exponent Attribute=true
    Create Position Attribute=true
    Create Radius Attribute=true
    Exponent=3
    Position=&MergeHelper_VectorSwitch(0=<0 -30 140>, 1=<0 30 140>, 2=?, 3=?, Index=$Particles.PartitionParticles.Partition, Num=::NewEnumerator0)
    Radius=60
    Scale Color=&Waveform(Global Clamp Min=0.5, "Waveform [1]"=::NewEnumerator2, "[1] Frequency"=1, "[1] Amplitude Min/Max"=<0 1>)
   )
  ]
  rendererClasses [&NiagaraLightRendererProperties]
  renderers [{rendererClass=&NiagaraLightRendererProperties, rendererIndex=0}]
 }
 VortexWind {
  emitterSpawnScript [
   /SetVariables(Emitter.RotationSpeed:NiagaraFloat=&FloatFromCurve(FloatCurve=DataInterface() {propertyValues {Curve {keys [<at 0 -0.5>,<at 0.1 0>]}}}, CurveIndex=$Emitter.NormalizedLoopAge, Scale Curve=0)) {isSetParameters=true}
   /SetVariables(
    Emitter.DarkColor:LinearColor=&FNTimeOfDayColor(
     Night Color=<rgba 0.6458 0.5867 0.5718 1>
     Morning Color=<rgba 0.6458 0.5943 0.5651 1>
     Day Color=<rgba 0.6458 0.5892 0.5449 1>
     Evening Color=<rgba 0.6458 0.6102 0.5819 1>
    )
    Emitter.MainColor:LinearColor=&FNTimeOfDayColor(
     Night Color=&SelectLinearColorFromArray(
      Array Sampling Mode=::NewEnumerator0
      Direct Array Index=&RandomRangeInt(Minimum=0, Maximum=1)
      Color Selection Array=DataInterface() {propertyValues {ColorData [<rgba 1 0.947 0.8646 1>]}}
      Random Seed=0
     )
     Morning Color=&SelectLinearColorFromArray(
      Array Sampling Mode=::NewEnumerator0
      Direct Array Index=&RandomRangeInt(Minimum=0, Maximum=1)
      Color Selection Array=DataInterface() {propertyValues {ColorData [<rgba 1 0.8815 0.8698 1>]}}
      Random Seed=0
     )
     Day Color=&SelectLinearColorFromArray(
      Array Sampling Mode=::NewEnumerator0
      Direct Array Index=&RandomRangeInt(Minimum=0, Maximum=1)
      Color Selection Array=DataInterface() {propertyValues {ColorData [<rgba 1 0.9425 0.901 1>, <rgba 1 0.9106 0.8802 1>]}}
      Random Seed=0
     )
     Evening Color=&SelectLinearColorFromArray(
      Array Sampling Mode=::NewEnumerator0
      Direct Array Index=&RandomRangeInt(Minimum=0, Maximum=1)
      Color Selection Array=DataInterface() {propertyValues {ColorData [<rgba 1 0.9332 0.8802 1>]}}
      Random Seed=0
     )
    )
   ) {isSetParameters=true}
  ]
  emitterUpdateScript [
   &EmitterState(Life Cycle Mode=::NewEnumerator1, Loop Behavior=::NewEnumerator1, Loop Duration=0.6, UseLoopDelay=true)
   &SpawnRate(SpawnRate=6, Use Spawn Probability=true)
  ]
  particleSpawnScript [
   &InitializeParticle(
    Color=<rgba 1 1 1 3>
    Lifetime=&RandomRangeFloat(Minimum=0.75, Maximum=1)
    Mesh Scale=&Multiply_VectorByFloat(Vector=<2 2 0.3>, Float=&RandomRangeFloat(Minimum=0.5, Maximum=0.65))
    Mesh Scale Mode=::NewEnumerator4
    Position Offset=<0 0 20>
    Position Offset Coordinate Space=::Local
    UsePositionOffset=true
    Write Material Random=true
   )
   &InitialMeshOrientation(
    Mesh Axis To Orient=<0 0 3>
    Mesh Orientation Mode=::NewEnumerator2
    Orientation Coordinate Space=::World
    Orientation Vector=<0 0 1>
    Rotation=&RandomRangeVector(Minimum=<0.03 0.03 0>, Maximum=<-0.03 -0.03 1>)
    Use Rotation Vector=true
   )
  ]
  particleUpdateScript [
   &ParticleState()
   &ScaleMeshSize(Scale Factor=&VectorFromFloat(Value=&FloatFromCurve(FloatCurve=DataInterface() {propertyValues {Curve {keys [<at 0 0.4137 RCIM_Cubic>,<at 1 0.7306>]}}})))
   &UpdateMeshOrientation(
    Rotation Rate=&FloatFromCurve(
     FloatCurve=DataInterface() {propertyValues {Curve {keys [<at 0.55 0.4376 RCIM_Cubic>, {arriveTangent=2.9123, interpMode=RCIM_Cubic, leaveTangent=2.9123, tangentMode=RCTM_User, time=1, value=1}]}}}
     CurveIndex=$Emitter.NormalizedLoopAge
     Scale Curve=-1.6154
    )
    Rotation Vector=<0 0 1>
   )
   &ScaleColor(
    Scale Alpha=&FloatFromCurve(
     FloatCurve=DataInterface() {propertyValues {Curve {keys [
         <at 0 0 RCIM_Cubic>
         {interpMode=RCIM_Cubic, tangentMode=RCTM_User, time=0.3258, value=1}
         {interpMode=RCIM_Cubic, tangentMode=RCTM_User, time=0.5806, value=1}
         {tangentMode=RCTM_User, time=1, value=0}
        ]}}}
     Scale Curve=&Multiply_Float(A=0.09, B=&FNTimeOfDayScalar(Night Scale=0.7, Morning Scale=0.9, Evening Scale=0.8))
    )
    Scale RGB=<2 2 2>
    ScaleA=true
    ScaleRGB=true
   )
   &DynamicMaterialParameters(
    Index 0 Param 1=&FloatFromCurve(FloatCurve=DataInterface() {propertyValues {Curve {keys [{tangentMode=RCTM_User, time=0.4, value=1}, {tangentMode=RCTM_User, time=0.9, value=-1}]}}})
    Index 0 Param 2=1
    Index 0 Param 3=&RandomRangeFloat(Minimum=-1, Maximum=1)
    Index 0 Param 4=&RandomRangeFloat(Minimum=-1, Maximum=1)
    Index 1 Param 1=&FloatFromCurve(
     FloatCurve=DataInterface() {propertyValues {Curve {keys [
         {arriveTangent=1.7366, interpMode=RCIM_Cubic, leaveTangent=1.7366, tangentMode=RCTM_User, time=0, value=0}
         {arriveTangent=0.4174, interpMode=RCIM_Cubic, leaveTangent=0.4174, tangentMode=RCTM_User, time=1, value=1}
        ]}}}
     Scale Curve=0.5
    )
    Index 1 Param 2=1
    Module.Index0Float0Write=true
    Module.Index0Float1Write=true
    Module.Index0Float2Write=true
    Module.Index0Float3Write=true
    Module.Index1Float0Write=true
    Module.Index1Float1Write=false
    Module.Index1Float2Write=false
    Module.Index1Float3Write=false
    Write Parameter Index 1=::NewEnumerator1
   ) {module=DynamicMaterialParameters001}
  ]
  rendererClasses [&NiagaraMeshRendererProperties]
  renderers [{rendererClass=&NiagaraMeshRendererProperties, rendererIndex=0}]
 }
}
summary {
 systemName=NS_Emote_SpruceBark_MotionBlur
 emitters [
  {emitterName=VortexWind, rendererClasses [&NiagaraMeshRendererProperties]}
  {emitterName=BlurRings_Detail, rendererClasses [&NiagaraMeshRendererProperties]}
  {emitterName=BlurRings_Thick, rendererClasses [&NiagaraMeshRendererProperties]}
  {emitterName=Light, rendererClasses [&NiagaraLightRendererProperties]}
 ]
 userVariables [User.Scale {type=NiagaraFloat, defaultValue=1, description=""}]
}
